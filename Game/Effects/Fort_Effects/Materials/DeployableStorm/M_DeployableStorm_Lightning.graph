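# Translucent, two-sided unlit material that draws animated lightning for the Deployable Storm effect, usable on skeletal meshes.
# Vertex offsets combine SplineThicken with near-camera width, stepped-time noise jitter, quality-switched curl noise and stepped axis rotation.
# Two panning smoke-trail samples pass through a moving, vertex-color-offset sphere mask, then get HDR tint and gradient-based hue shift.
# Multiplied sine waves produce irregular flicker, including thresholded on/off values and vertex-color channel selection.
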Material /Game/Effects/Fort_Effects/Materials/DeployableStorm/M_DeployableStorm_Lightning
graph {
 properties {
  bUsedWithMeshDeformer=true
  bUsedWithSkeletalMesh=true
  blendMode=BLEND_Translucent
  pixelDepthOffsetMode=PDOM_Legacy
  shadingModel=MSM_Unlit
  twoSided=true
  usedShadingModels=Unlit
 }
 nodes [
  n1 = Subtract(A=ComponentMask(None=TextureCoordinate()) {g=true}, B=Constant()) {constB=0.5}
  n2 = Clamp(None=Multiply(A=@n1, B=OneMinus(None=@n1)))
  n3 = Add(A=TextureSample(UVs=TextureCoordinate() {vTiling=2}) {texture=/Game/Effects/Fort_Effects/Textures/Ramps/T_Ramp_Smooth_Reverse}.G) {constB=0.6}
  n4 = Power(Base=MaterialFunctionCall(Distance=ScalarParameter() {defaultValue=25, parameterName=NearFade}) {materialFunction=&Near_Camera_Fade}.Result)
  n5 = Multiply(
   A=Multiply(A=@n2, B=Clamp(None=@n3))
   B=LinearInterpolate(
    A=ScalarParameter() {parameterName=SplineThickness_Close}
    B=ScalarParameter() {defaultValue=100, parameterName=SplineThickness}
    Alpha=Clamp(None=@n4)
   )
  )
  n6 = MaterialFunctionCall(WidthBase=@n5, WidthTip=@n5, UVs for Thickness=TextureCoordinate()) {materialFunction=&SplineThicken}
  n7 = TextureCoordinate() {uTiling=0.002, vTiling=0.1}
  n8 = ScalarParameter() {defaultValue=5, parameterName=Tile}
  n9 = Time()
  n10 = ScalarParameter() {defaultValue=15, parameterName=PanStepSize}
  n11 = Divide(A=Round(None=Multiply(A=@n9, B=@n10)), B=@n10)
  n12 = Constant2Vector() {g=1, r=0.25}
  n13 = ScalarParameter() {defaultValue=-1, parameterName=PanSpeed}
  n14 = TextureSample(UVs=Panner(Coordinate=Multiply(A=@n7, B=@n8), Time=@n11, Speed=Multiply(A=@n12, B=@n13)) {bFractionalPart=true, speedX=0.01, speedY=0.1}) {samplerType=SAMPLERTYPE_LinearColor, texture=/Game/Global/Textures/Default/Noise/Noise32x32}
  n15 = OneMinus(None=@n14.RGB)
  n16 = Panner(
   Coordinate=Multiply(A=TextureCoordinate() {vTiling=0.1}, B=Multiply(A=@n8) {constB=6}) {constB=0.1}
   Time=Multiply(A=@n11) {constB=0.5}
   Speed=Multiply(A=@n13) {constB=-1.43}
  ) {bFractionalPart=true, speedX=-0.05}
  n17 = Multiply(
   A=Multiply(A=Subtract(A=@n14.RGB, B=@n15), B=ScalarParameter() {defaultValue=5, parameterName=WPOMult}) {constA=1}
   B=Add(A=TextureSample(UVs=@n16) {texture=/Game/Global/Textures/Default/Masks_and_Gradients/T_Vert_Gradient_Sections}.R) {constB=0.15}
  )
  n18 = WorldPosition()
  n19 = ActorPositionWS()
  n20 = ScalarParameter() {defaultValue=25, parameterName=CurlSpeed}
  n21 = Time()
  n22 = Divide(A=Subtract(A=Subtract(A=@n18.XYZ, B=@n19), B=Multiply(A=@n20, B=@n21)), B=ScalarParameter() {defaultValue=25, parameterName=CurlScale})
  n23 = QualitySwitch(
   Default=Multiply(A=ScalarParameter() {defaultValue=50, parameterName=CurlIntensity}, B=VectorNoise(World Position=@n22) {noiseFunction=VNF_VectorALU})
   Low=Constant() {r=1}
  )
  n24 = Transform(None=VectorParameter() {parameterName=RotationVector, defaultValue=<rgba 0 0 1 1>}.RGB) {transformSourceType=TRANSFORMSOURCE_Local}
  n25 = ScalarParameter() {defaultValue=2, parameterName=RotatePicker}
  n26 = Sine(None=Time()) {period=3}
  n27 = Sine(None=Time()) {period=4}
  n28 = Time()
  n29 = LinearInterpolate(Alpha=Multiply(A=Multiply(A=@n26, B=@n27), B=Sine(None=@n28) {period=4}))
  n30 = Divide(A=Ceil(None=Multiply(A=@n25, B=@n29)), B=@n25) {constB=1}
  Add(
   A=Add(A=Add(A=@n6.WorldPositionOffset, B=@n17), B=Multiply(A=@n2, B=@n23))
   B=RotateAboutAxis(
    NormalizedRotationAxis=Normalize(VectorInput=@n24)
    RotationAngle=@n30
    PivotPoint=MaterialFunctionCall() {materialFunction=&ObjectPivotPoint}.Object Pivot Location
    Position=WorldPosition().XYZ
   )
  )
  n31 = Sine(None=Time()) {period=4.3}
  n32 = Sine(None=Time()) {period=2.5}
  n33 = Time()
  n34 = LinearInterpolate(Alpha=Multiply(A=Multiply(A=@n31, B=@n32), B=Sine(None=@n33) {period=1.46}))
  n35 = Constant()
  If(A=@n34, B=Constant() {r=0.1}, "A > B"=Constant() {r=1}, "A == B"=@n35, "A < B"=@n35)
  n36 = Sine(None=Time()) {period=2.1}
  n37 = Sine(None=Time()) {period=2.1}
  n38 = Time()
  n39 = LinearInterpolate(Alpha=Multiply(A=Multiply(A=@n36, B=@n37), B=Sine(None=@n38) {period=1.2}))
  n40 = Constant()
  If(A=@n39, B=Constant() {r=0.4}, "A > B"=Constant() {r=1}, "A == B"=@n40, "A < B"=@n40)
  n41 = Sine(None=Time())
  n42 = Sine(None=Time()) {period=1.25}
  n43 = Time()
  n44 = LinearInterpolate(Alpha=Multiply(A=Multiply(A=@n41, B=@n42), B=Sine(None=@n43) {period=6}))
  n45 = Constant()
  If(A=@n44, B=Constant() {r=0.2}, "A > B"=Constant() {r=1}, "A == B"=@n45, "A < B"=@n45)
  n46 = VertexColor()
  n47 = Multiply(A=Floor(None=Add(B=@n46.G) {constA=0.4}), B=@n34)
  n48 = Add(B=VertexColor().R) {constA=0.4}
  n49 = Add(B=VertexColor().B) {constA=0.4}
  n50 = ScalarParameter() {defaultValue=3, parameterName=BlendMultPicker}
  n51 = Time()
  n52 = Time()
  n53 = Multiply(A=Multiply(A=Sine(None=@n51) {period=3}, B=Sine(None=@n52) {period=2.6}), B=Sine(None=Time()) {period=2.15})
  n54 = Constant()
  n55 = Multiply(A=@n50, B=If(A=LinearInterpolate(Alpha=@n53), B=Constant() {r=0.5}, "A > B"=Constant() {r=1}, "A == B"=@n54, "A < B"=@n54))
  MaterialFunctionCall(A=@n47, B=Multiply(A=Floor(None=@n48), B=@n39), C=Multiply(A=Floor(None=@n49), B=@n44), Alpha=Divide(A=Ceil(None=@n55), B=@n50) {constB=1}) {materialFunction=&3ColorBlend}
  n56 = MaterialFunctionCall() {materialFunction=&LinearGradient}
  n57 = OneMinus(None=@n56.VGradient)
  n58 = TextureCoordinate() {coordinateIndex=1}
  n59 = ComponentMask(None=VectorParameter() {parameterName=Electrical_Tiling, defaultValue=<rgba 1 1 0 0>}.RGB) {g=true, r=true}
  n60 = ComponentMask(None=VectorParameter() {parameterName=Electrical_Speed, defaultValue=<rgba 0 1 0 0>}.RGB) {g=true, r=true}
  n61 = MaterialFunctionCall(
   UVs=Panner(Coordinate=Multiply(A=@n58, B=@n59), Speed=@n60) {speedY=1}
   "Rotation Angle (0-1)"=ScalarParameter() {defaultValue=0.25, parameterName="Rotation Angle (0-1) (S)"}
  ) {materialFunction=&CustomRotator}
  n62 = TextureCoordinate() {coordinateIndex=1}
  n63 = ComponentMask(None=VectorParameter() {parameterName=Electrical_Tiling_1, defaultValue=<rgba 1 2.5 0 0>}.RGB) {g=true, r=true}
  n64 = MaterialFunctionCall(
   UVs=Panner(Coordinate=Multiply(A=@n62, B=@n63), Speed=@n60) {speedY=1}
   "Rotation Angle (0-1)"=ScalarParameter() {defaultValue=0.25, parameterName="Rotation Angle (0-1) (S)_1"}
  ) {materialFunction=&CustomRotator}
  n65 = Constant()
  n66 = VertexColor()
  n67 = Add(A=VertexColor().R, B=Multiply(A=@n66.G) {constB=0.25})
  n68 = Constant()
  n69 = ScalarParameter() {defaultValue=0.146, parameterName="Speed (S)"}
  n70 = ScalarParameter() {defaultValue=0.5, parameterName="Speed2 (S)"}
  n71 = LinearInterpolate(
   A=@n66.G
   B=Multiply(A=@n66.G, B=ScalarParameter() {defaultValue=0.25, parameterName=Param_1})
   Alpha=LinearInterpolate(
    A=MaterialFunctionCall(Speed=@n69) {materialFunction=&DebugTimeSine}.Result
    B=MaterialFunctionCall(Speed=@n70) {materialFunction=&DebugTimeSine}.Result
   )
  )
  n72 = Time() {bOverride_Period=true, period=5}
  n73 = VertexColor()
  n74 = MaterialFunctionCall(Speed=ScalarParameter() {defaultValue=0.25, parameterName=PanSpeed_Sine}) {materialFunction=&DebugTimeSine}
  n75 = Multiply(
   A=Frac(None=Add(A=@n71, B=@n72))
   B=LinearInterpolate(
    A=ScalarParameter() {defaultValue=2.25, parameterName=TimeMult}
    B=ScalarParameter() {defaultValue=1, parameterName=TimeMult_High}
    Alpha=Multiply(A=@n73.G, B=@n74.Result)
   )
  )
  n76 = ScalarParameter() {defaultValue=0.35, parameterName=Panning_Mask_Radius}
  n77 = ScalarParameter() {defaultValue=0.5, parameterName=Panning_Mask_Radius_Max}
  n78 = LinearInterpolate(Alpha=ScalarParameter() {defaultValue=1, parameterName=RandomOffsetPerClient_Mesh})
  n79 = Add(A=Time() {bOverride_Period=true, period=2.6}, B=@n78)
  n80 = Add(A=Time() {bOverride_Period=true, period=1.57}, B=@n78)
  n81 = Constant() {r=0.15}
  n82 = Constant() {r=0.85}
  n83 = Sine(None=Add(A=Time() {bOverride_Period=true, period=1.24}, B=@n78)) {period=1.24}
  n84 = Multiply(
   A=LinearInterpolate(
    A=Sine(None=@n79) {period=2.6}
    B=Sine(None=@n80) {period=1.57}
    Alpha=LinearInterpolate(A=@n81, B=@n82, Alpha=@n83) {constA=0.15, constB=0.85}
   )
   B=VertexColor().B
  )
  n85 = Multiply(
   A=Add(
    A=TextureSample(UVs=@n61.Rotated Values) {samplerType=SAMPLERTYPE_Grayscale, texture=&T_Smoke_Trail_Soft}.G
    B=TextureSample(UVs=@n64.Rotated Values) {samplerType=SAMPLERTYPE_Grayscale, texture=&T_Smoke_Trail_Soft}.G
   )
   B=SphereMask(
    A=AppendVector(A=@n65, B=@n67)
    B=AppendVector(A=@n68, B=@n75)
    Radius=LinearInterpolate(A=@n76, B=@n77, Alpha=@n84)
    Hardness=ScalarParameter() {defaultValue=0.1, parameterName=Panning_Mask_Hardness}
   )
  )
  MaterialFunctionCall(
   Hue Shift Percentage=Multiply(A=Multiply(A=@n56.VGradient, B=@n57), B=ScalarParameter() {defaultValue=0.2, parameterName=HueShift})
   Texture=MaterialFunctionCall(LDR Input=Saturate(None=@n85), HDR Tint=VectorParameter() {parameterName=Emissive Color, defaultValue=<rgba 5 0 25 0>}.RGB) {materialFunction=&DeriveHDRfromLDR}.RemappedRange
  ) {materialFunction=&HueShift}
  MaterialFunctionCall!(Rotation Amount=@n30, PivotPoint, WorldPosition) {materialFunction=/Engine/Functions/Engine_MaterialFunctions02/WorldPositionOffset/RotateAboutWorldAxis_cheap}
  n86 = MaterialFunctionCall() {materialFunction=&LinearGradient}
  Multiply(A=Saturate(None=Multiply(A=@n47) {constB=300}), B=Multiply(A=@n86.UGradient, B=OneMinus(None=@n86.UGradient)))
  n87 = TextureCoordinate() {coordinateIndex=1}
  n88 = Multiply(A=TextureCoordinate() {coordinateIndex=1, uTiling=0}, B=ScalarParameter() {defaultValue=16, parameterName=NumberOfBars})
  n89 = MaterialFunctionCall!(
   PrimaryTextureCoordinates=@n88
   2ndOrderTextureCoordinates=@n88
   RandomSeed=ScalarParameter() {defaultValue=15, parameterName=Speed}
   NoiseDensityBias
   Bypass Noise Influence Crosstalk
  ) {materialFunction=/Game/Effects/Fort_Effects/Materials/_Functions/MF_RandomUVSample}
  n90 = ComponentMask(None=Multiply(A=VectorParameter() {parameterName=ElectricalMask_Speed, defaultValue=<rgba 0 5 0 0>}.RGB, B=@n89.2ndOrderNoiseWithRange)) {g=true, r=true}
  n91 = TextureCoordinate() {coordinateIndex=1}
  n92 = ComponentMask(None=Multiply(A=@n89.2ndOrderNoiseWithRange, B=VectorParameter() {parameterName=ElectricalMask_Speed_2, defaultValue=<rgba 0 4.1 0 0>}.RGB)) {g=true, r=true}
  Multiply(
   A=TextureSample(UVs=Panner(Coordinate=@n87, Speed=@n90) {speedY=0.7}) {texture=/Game/Global/Textures/Default/Masks_and_Gradients/T_Water_Top_Ramp}.G
   B=TextureSample(UVs=Panner(Coordinate=@n91, Speed=@n92) {speedY=1}) {texture=&T_Water_Top_Ramp}.G
  )
  n93 = MaterialFunctionCall() {materialFunction=&LinearGradient}
  Multiply(A=@n85, B=Multiply(A=@n93.UGradient, B=OneMinus(None=@n93.UGradient)))
  ScalarParameter() {defaultValue=1, parameterName=Opacity}
 ]
}
