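# Unlit, two-sided translucent material for the Deployable Storm effect's lightning smoke trails.
# Panning smoke textures drive opacity, faded along V, and a dark-to-purple-blue tint lit by time-of-day color, Schlick-phase sun scattering and sine-driven lightning flicker, with V-masked hue shift.
# World position offset combines SplineThicken width, panning texture waves and a directional bend.
# Low quality, mobile and forward paths drop the waves; Low and forward also use single-layer opacity.
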Material /Game/Effects/Fort_Effects/Materials/DeployableStorm/M_DeployableStorm_Lightning_Smoke
graph {
 properties {
  bUsedWithMeshDeformer=true
  bUsedWithSkeletalMesh=true
  blendMode=BLEND_Translucent
  pixelDepthOffsetMode=PDOM_Legacy
  shadingModel=MSM_Unlit
  twoSided=true
  usedShadingModels=Unlit
 }
 nodes [
  n1 = OneMinus(None=ComponentMask(None=TextureCoordinate()) {g=true})
  n2 = Transform(None=VectorParameter() {parameterName=BendDirection, defaultValue=<rgba 1 0 0 1>}.RGB) {transformSourceType=TRANSFORMSOURCE_Local}
  n3 = MaterialFunctionCall(
   WidthBase=ScalarParameter() {defaultValue=20, group=WPO, parameterName=WidthBase_Max}
   WidthTip=ScalarParameter() {defaultValue=10, group=WPO, parameterName=WidthTip_Max}
  ) {materialFunction=&SplineThicken}
  n4 = Constant3Vector()
  n5 = VectorParameter() {group=WPO, parameterName=TextureWPO_UVtiling, defaultValue=<rgba 1 1 0 0>}
  n6 = Panner(Coordinate=Multiply(A=TextureCoordinate(), B=ComponentMask(None=@n5.RGB) {g=true, r=true})) {speedY=2}
  n7 = TextureSample(UVs=MaterialFunctionCall(UVs=@n6, "Rotation Angle (0-1)"=ScalarParameter() {defaultValue=0.25, group=Texture, parameterName="Rotation Angle (0-1) (S)"}) {materialFunction=&CustomRotator}.Rotated Values) {samplerType=SAMPLERTYPE_Normal, texture=/Game/Effects/Fort_Effects/Textures/Smoke/T_Smoke_Trail_Soft_N}
  n8 = ComponentMask(None=@n7.RGB) {r=true}
  n9 = ScalarParameter() {defaultValue=1, group=WPO, parameterName=WPO_HeightX}
  n10 = ComponentMask(None=@n7.RGB) {g=true}
  n11 = ScalarParameter() {defaultValue=1, group=WPO, parameterName=WPO_HeightY}
  n12 = AppendVector(
   A=AppendVector(A=Multiply(A=@n8, B=@n9) {constB=0.5}, B=Multiply(A=@n10, B=@n11) {constB=0.5})
   B=Multiply(
    A=TextureSample(UVs=@n6) {texture=&T_Smooth_tile}.G
    B=ScalarParameter() {defaultValue=1, group=WPO, parameterName=WPO_HeightZ}
   ) {constB=0.6}
  )
  n13 = Constant() {r=0.5}
  n14 = SphereMask(
   A=MaterialFunctionCall() {materialFunction=&LinearGradient}.VGradient
   B=@n13
   Radius=@n13
  ) {hardnessPercent=0}
  n15 = Multiply(
   A=Transform(None=LinearInterpolate(A=@n4, B=@n12, Alpha=@n14))
   B=ScalarParameter() {defaultValue=5, group=WPO, parameterName=WPO_Wave_Intenstiy}
  ) {constB=80}
  n16 = Constant3Vector() {desc=Disable WPO anim in Low, constant=<rgba 0 0 20 0>}
  n17 = ShadingPathSwitch(Default=@n15, Deferred=@n15, Mobile=@n16) {desc=Migrated from FeatureLevelSwitch}
  n18 = QualitySwitch(Default=Add(A=@n3.WorldPositionOffset, B=QualitySwitch(Default=@n17, Low=@n16, High=@n17, Medium=@n17)), Low=@n3.WorldPositionOffset)
  Add(
   A=Multiply(A=ScalarParameter() {defaultValue=5, parameterName=BendIntensity}, B=Multiply(A=@n1, B=@n2))
   B=ShadingPathSwitch(
    Default=ShadingPathSwitch(Default=@n18, Mobile=@n3.WorldPositionOffset) {desc=Migrated from FeatureLevelSwitch}
    Forward=@n3.WorldPositionOffset
   )
  )
  n19 = CollectionParameter() {
   collection=&FortniteMaterialParameters
   parameterName=SunlightDirectionalVector
  }
  MaterialExpressionDotProduct_1 = DotProduct(A=ComponentMask(None=@n19) {b=true, g=true, r=true}, B=CameraVectorWS())
  MaterialExpressionScalarParameter_5 = ScalarParameter() {defaultValue=0.4, parameterName=Anisotropy}
  n20 = Custom!(CosTheta=@MaterialExpressionDotProduct_1, k=@MaterialExpressionScalarParameter_5) {
   code="\tfloat Inner = (1 + k * CosTheta);\r\n\treturn (1 - k * k) / (4 * PI * Inner * Inner);\r\n\r\n"
   description=Schlick Phase
   inputs [
    {
     inputName=CosTheta
     input {expression=&self:MaterialExpressionDotProduct_1}
    }
    {
     inputName=k
     input {expression=&self:MaterialExpressionScalarParameter_5}
    }
   ]
  }
  n21 = MaterialFunctionCall(
   Color_Morning=VectorParameter() {group=Color, parameterName="Color_Morning (V3)", defaultValue=<rgba 1.25 1.25 1.25 0>}.RGB
   Color_Day=VectorParameter() {group=Color, parameterName="Color_Day (V3)", defaultValue=<rgba 1.25 1.25 1.25 0>}.RGB
   Color_Evening=VectorParameter() {group=Color, parameterName="Color_Evening (V3)", defaultValue=<rgba 1 1 1 0>}.RGB
   Color_Night=VectorParameter() {group=Color, parameterName="Color_Night (V3)", defaultValue=<rgba 1 1 1 0>}.RGB
  ) {materialFunction=&Color_Time_of_Day}
  n22 = Multiply(A=Multiply(A=@n20, B=ScalarParameter() {defaultValue=25, parameterName=ScaleCloudColor}), B=Multiply(A=@n21.Result))
  n23 = LinearInterpolate(Alpha=ScalarParameter() {defaultValue=1, parameterName=RandomOffsetPerClient_Mesh})
  n24 = Add(A=Time() {bOverride_Period=true, period=1}, B=@n23)
  n25 = Add(A=Time() {bOverride_Period=true, period=0.35}, B=@n23)
  n26 = Constant() {r=0.15}
  n27 = Constant() {r=0.85}
  n28 = Sine(None=Add(A=Time() {bOverride_Period=true, period=0.6}, B=@n23)) {period=1.24}
  n29 = Saturate(None=LinearInterpolate(
    A=Sine(None=@n24) {period=2.6}
    B=Sine(None=@n25) {period=1.57}
    Alpha=LinearInterpolate(A=@n26, B=@n27, Alpha=@n28) {constA=0.15, constB=0.85}
   ))
  n30 = ScalarParameter() {defaultValue=10, parameterName=FlashMult}
  n31 = TextureCoordinate() {vTiling=0.5}
  n32 = VectorParameter() {group=Texture, parameterName=PannerSpeed, defaultValue=<rgba 0 0.5 0 1>}
  n33 = MaterialFunctionCall(
   UVs=Panner(Coordinate=@n31, Speed=ComponentMask(None=@n32.RGB) {g=true, r=true})
   "Rotation Angle (0-1)"=ScalarParameter() {defaultValue=0.25, group=Texture, parameterName="Rotation Angle (0-1) (S)"}
  ) {materialFunction=&CustomRotator}
  n34 = Power(
   Base=Multiply(A=TextureSample(UVs=@n33.Rotated Values) {texture=/Game/Effects/Fort_Effects/Textures/Smoke/T_Smoke_Trail_Soft_2}.B)
   Exp=ScalarParameter() {defaultValue=1, parameterName=Exp}
  )
  n35 = TextureSample(UVs=@n33.Rotated Values) {texture=&T_Smoke_Trail_Soft_2}
  n36 = AppendVector(A=Constant() {r=1}, B=ScalarParameter() {defaultValue=1.25, group=Texture, parameterName=TilingSmoke_V_Mult})
  n37 = VectorParameter() {group=Texture, parameterName=PannerSpeed_02, defaultValue=<rgba 0 0.6 0 1>}
  n38 = MaterialFunctionCall(
   UVs=Panner(Coordinate=Multiply(A=@n31, B=@n36), Speed=ComponentMask(None=@n37.RGB) {g=true, r=true})
   "Rotation Angle (0-1)"=ScalarParameter() {defaultValue=0.25, group=Texture, parameterName="Rotation Angle (0-1) (S)"}
  ) {materialFunction=&CustomRotator}
  n39 = Multiply(
   A=Add(A=@n35.B, B=TextureSample(UVs=@n38.Rotated Values) {texture=&T_Smoke_Trail_Soft_2}.B)
   B=ScalarParameter() {defaultValue=1, group=Texture, parameterName=Opacity}
  )
  n40 = Multiply(A=@n35.B, B=ScalarParameter() {defaultValue=1, group=Texture, parameterName=OpacityLowEnd})
  n41 = Power(
   Base=ShadingPathSwitch(Default=QualitySwitch(Default=@n39, Low=@n40), Forward=@n40)
   Exp=ScalarParameter() {defaultValue=1, parameterName=Exp}
  )
  n42 = Multiply(
   A=Add(A=Multiply(A=@n22, B=@n21.Result), B=Multiply(A=@n29, B=@n30))
   B=LinearInterpolate(
    A=VectorParameter() {parameterName=ColorDark, defaultValue=<rgba 0.05 0 0.05 0>}.RGB
    B=VectorParameter() {parameterName=Color, defaultValue=<rgba 0.0353 0 0.3411 0>}.RGB
    Alpha=Multiply(A=@n34, B=@n41)
   )
  )
  n43 = TextureCoordinate()
  n44 = ComponentMask(None=@n43) {g=true}
  n45 = ComponentMask(None=@n43) {g=true}
  LinearInterpolate(
   A=MaterialFunctionCall(Hue Shift Percentage=ScalarParameter() {defaultValue=0.05, parameterName="Hue Shift Percentage (S)"}, Texture=@n42) {materialFunction=&HueShift}.Result
   B=@n42
   Alpha=Multiply(A=Multiply(A=@n44, B=@n44), B=OneMinus(None=@n45))
  )
  n46 = TextureCoordinate()
  n47 = ComponentMask(None=@n46) {g=true}
  n48 = ComponentMask(None=@n46) {g=true}
  Multiply(A=Multiply(A=Multiply(A=@n47, B=@n47), B=OneMinus(None=@n48)), B=@n41)
 ]
}
